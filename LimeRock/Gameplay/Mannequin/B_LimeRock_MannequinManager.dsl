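; LimeRock mannequin manager that dresses four mannequins as the local player and squadmates, apparently for a sequenced event.
; On clients, playlist-ready or Initialize Mannequins gathers living same-squad player states, ordered by squad member index, with a gated 30-second delayed re-init as fallback.
; SetupMannequins copies each pawn's character, backpack and cosmetic variants onto its mannequin, leaving placement to Sequencer.
; Shipping-log helpers report setup problems and mannequin location and visibility.
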(event EventBeginPlay
  (bind _teammembers (Variables|Transient|GetTeamMembers))
  ;[0 0.15 0.03648 0.5] BeginPlay: Populating the mannequin setup.
  (Class|FortAsyncActionCurrentPlaylistReady|PlaylistDataReadyAsync
    (:OnReady
      (|CallParentFunction)
      (CallFunction|StartPrintingMannequinDebug)
      (if (not (Networking|IsServer))
        (Utilities|FlowControl|Sequence
          (:then_0
            _)
          (:then_1
            (Utilities|FlowControl|Delay 3e+01) ; trigger mannequin init again on client\nafter 30 seconds to just really guanantee\ninit. Sequencer also calls this, but reaching\nPencils Down and having launcher problems.
            (Utilities|FlowControl|Gate false :!label _label
              (:Exit
                (Utilities|FlowControl|Sequence
                  (:then_0
                    (goto _label Close))
                  (:then_1
                    _))))))
        (Utilities|Array|Clear (Variables|Transient|GetTeamMembers))
        (CallFunction|LogIt "Grabbing squad list and initializing mannequins. This may happen multiple times.")
        (bind _returnvalue (Fortnite|GetFirstLocalFortPlayerController true))
        (Utilities|IsValid _returnvalue
          (:Is%20Valid
            (bind _asfort_player_state_athena (Utilities|Casting|CastToFortPlayerStateAthena (Variables|/Script/Engine.Controller|GetPlayerState _returnvalue)))
            (bind _returnvalue_2 (Class|FortPlayerStateAthena|GetSquadId _asfort_player_state_athena))
            (bind _returnvalue_3 (Fortnite|Athena|GameState|GetGameStateAthena))
            (for _array_element (Variables|/Script/Engine.GameStateBase|GetPlayerArray _returnvalue_3)
              (bind _asfort_player_state_athena_2 (Utilities|Casting|CastToFortPlayerStateAthena _array_element))
              (bind _returnvalue_4 (Class|FortPlayerStateAthena|GetSquadId _asfort_player_state_athena_2))
              (if (== _returnvalue_4 _returnvalue_2)
                (bind _asfort_player_pawn (Utilities|Casting|CastToFortPlayerPawn (PlayerState|GetPawn _asfort_player_state_athena_2)))
                (if (not (FortPawn|IsDead _asfort_player_pawn))
                  (Utilities|Array|SetArrayElem (Variables|Transient|GetTeamMembers) (Class|FortPlayerStateAthena|GetSquadMemberIndex _asfort_player_state_athena_2) _asfort_player_state_athena_2 true))))
            (CallFunction|LogIt (Utilities|String|Append "Total found team members: " (Utilities|String|ToString(Integer) (Utilities|Array|Length (Variables|Transient|GetTeamMembers)))))
            (Utilities|Array|RemoveItem _teammembers) ; Remove empty spaces if any.
            (CallFunction|SetupMannequins (Utilities|Array|Length _teammembers) _teammembers))
          (:Is%20Not%20Valid
            (CallFunction|LogIt "Didn't find Local Player Controller" true)))))))

(event Custom|Initialize%20Mannequins
  (bind _teammembers (Variables|Transient|GetTeamMembers))
  ;[0 0.15 0.03648 0.5] BeginPlay: Populating the mannequin setup.
  (if (not (Networking|IsServer))
    (Utilities|FlowControl|Sequence
      (:then_0
        _)
      (:then_1
        (Utilities|FlowControl|Delay 3e+01) ; trigger mannequin init again on client\nafter 30 seconds to just really guanantee\ninit. Sequencer also calls this, but reaching\nPencils Down and having launcher problems.
        (Utilities|FlowControl|Gate false :!label _label
          (:Exit
            (Utilities|FlowControl|Sequence
              (:then_0
                (goto _label Close))
              (:then_1
                _))))))
    (Utilities|Array|Clear (Variables|Transient|GetTeamMembers))
    (CallFunction|LogIt "Grabbing squad list and initializing mannequins. This may happen multiple times.")
    (bind _returnvalue (Fortnite|GetFirstLocalFortPlayerController true))
    (Utilities|IsValid _returnvalue
      (:Is%20Valid
        (bind _asfort_player_state_athena (Utilities|Casting|CastToFortPlayerStateAthena (Variables|/Script/Engine.Controller|GetPlayerState _returnvalue)))
        (bind _returnvalue_2 (Class|FortPlayerStateAthena|GetSquadId _asfort_player_state_athena))
        (bind _returnvalue_3 (Fortnite|Athena|GameState|GetGameStateAthena))
        (for _array_element (Variables|/Script/Engine.GameStateBase|GetPlayerArray _returnvalue_3)
          (bind _asfort_player_state_athena_2 (Utilities|Casting|CastToFortPlayerStateAthena _array_element))
          (bind _returnvalue_4 (Class|FortPlayerStateAthena|GetSquadId _asfort_player_state_athena_2))
          (if (== _returnvalue_4 _returnvalue_2)
            (bind _asfort_player_pawn (Utilities|Casting|CastToFortPlayerPawn (PlayerState|GetPawn _asfort_player_state_athena_2)))
            (if (not (FortPawn|IsDead _asfort_player_pawn))
              (Utilities|Array|SetArrayElem (Variables|Transient|GetTeamMembers) (Class|FortPlayerStateAthena|GetSquadMemberIndex _asfort_player_state_athena_2) _asfort_player_state_athena_2 true))))
        (CallFunction|LogIt (Utilities|String|Append "Total found team members: " (Utilities|String|ToString(Integer) (Utilities|Array|Length (Variables|Transient|GetTeamMembers)))))
        (Utilities|Array|RemoveItem _teammembers) ; Remove empty spaces if any.
        (CallFunction|SetupMannequins (Utilities|Array|Length _teammembers) _teammembers))
      (:Is%20Not%20Valid
        (CallFunction|LogIt "Didn't find Local Player Controller" true)))))

(event EventEndPlay
  (params (EndPlayReason :type EEndPlayReason%20Enum :default "Destroyed"))
  ;[0 0.15 0.03648 0.5] BeginPlay: Populating the mannequin setup.
  (Utilities|Time|ClearandInvalidateTimerbyHandle (Variables|Default|GetDebug%20Timer%20Handle)))

(event Custom|Start%20Printing%20Mannequin%20Debug)

(event Custom|Print%20Mannequin%20Debug
  ; Debug Timer
  (CallFunction|LogIt (Utilities|String|Append "[Debug Timer] " (CallFunction|MannequinDebugString (Variables|Mannequin|GetMannequin%20-%20Local%20Player) "player") (CallFunction|MannequinDebugString (Variables|Mannequin|GetMannequin%20-%2002) "02") (CallFunction|MannequinDebugString (Variables|Mannequin|GetMannequin%20-%2003) "03") (CallFunction|MannequinDebugString (Variables|Mannequin|GetMannequin%20-%2004) "04"))))

(fn LogIt
  (params (string :type String) (warning :type Boolean))
  (Fortnite|FortShippingLog 0 (Utilities|String|Append "[LimeRock Mannequin Manager] " (Utilities|String|SelectString "(server)" "(client)" (Networking|IsServer)) " - " string) warning))

(fn MannequinDebugString
  (params (mannequin :type Fort%20Player%20Mannequin%20Object%20Reference) (index :type String))
  (return (Utilities|String|Append "  **[Mannequin " index "] location: " (Utilities|String|ToString(Vector) (Transformation|GetActorLocation mannequin)) ", " (Utilities|String|SelectString "hidden" "visible" (Variables|Rendering|GetbHidden mannequin)) "**  ")))

(fn SetupMannequins
  (params (Count :type Integer) (TeamMembers :type Array%20of%20Fort%20Player%20State%20Athena%20Object%20References))
  (bind _mannequin___04 (Variables|Mannequin|GetMannequin%20-%2004))
  (bind _mannequin___03 (Variables|Mannequin|GetMannequin%20-%2003))
  (bind _mannequin___02 (Variables|Mannequin|GetMannequin%20-%2002))
  (bind _mannequin___local_player (Variables|Mannequin|GetMannequin%20-%20Local%20Player))
  (bind _array (Utilities|Array|MakeArray _mannequin___local_player _mannequin___02 _mannequin___03 _mannequin___04))
  (Utilities|FlowControl|Sequence
    (:then_1
      (CallFunction|LogIt (Utilities|String|Append "SetupMannequins - iterating through this many squadmates: " (Utilities|String|ToString(Integer) (Utilities|Array|Length TeamMembers))))
      (for (_array_element _array_index) TeamMembers
        (Utilities|IsValid _array_element
          (:Is%20Valid
            (if (Utilities|Array|IsValidIndex _array _array_index)
              (Utilities|IsValid (PlayerState|GetCurrentPawn _array_element)
                (:Is%20Valid
                  (Utilities|IsValid (Utilities|Array|Get(acopy) _array _array_index)
                    (:Is%20Valid
                      (Customization|SetAthenaBackpackandCharacterDefandVariants (Utilities|Array|Get(acopy) _array _array_index) (CosmeticLoadout|GetItemDefinitionTypefromLoadoutwithActor (PlayerState|GetCurrentPawn _array_element) "(TagName=\"LoadoutSlot.Backpack\")" "/Script/FortniteGame.AthenaBackpackItemDefinition") (Class|FortCosmeticFunctionLibrary|GetActiveVariantsforItemUsingExistingVariants (CosmeticLoadout|GetItemDefinitionTypefromLoadoutwithActor (PlayerState|GetCurrentPawn _array_element) "(TagName=\"LoadoutSlot.Backpack\")" "/Script/FortniteGame.AthenaBackpackItemDefinition") (CosmeticLoadout|GetAllLoadoutVariantInfofromActor (PlayerState|GetCurrentPawn _array_element))) (CosmeticLoadout|GetItemDefinitionTypefromLoadoutwithActor (PlayerState|GetCurrentPawn _array_element) "(TagName=\"LoadoutSlot.Character\")" "/Script/FortniteGame.AthenaCharacterItemDefinition") (CosmeticLoadout|GetAllLoadoutVariantInfofromActor (PlayerState|GetCurrentPawn _array_element)))
                      (CallFunction|LogIt (Utilities|String|Append "Set Athena Backpack and Character Def and Variants. This many variants: " (Utilities|String|ToString(Integer) (Utilities|Array|Length (CosmeticLoadout|GetAllLoadoutVariantInfofromActor (PlayerState|GetCurrentPawn _array_element)))))))
                    (:Is%20Not%20Valid
                      (CallFunction|LogIt (Utilities|String|Append "SetupMannequins, Mannequin not valid: " (Utilities|String|ToString(Integer) _array_index)) true))))
                (:Is%20Not%20Valid
                  (CallFunction|LogIt (Utilities|String|Append "SetupMannequins, Didn't find a valid pawn: " (Utilities|String|ToString(Integer) _array_index)) true)))))
          (:Is%20Not%20Valid
            (CallFunction|LogIt (Utilities|String|Append "SetupMannequins, squadmate not valid: " (Utilities|String|ToString(Integer) _array_index)) true)))))) ; 0 - disconnected, let sequencer handle placement \n+ mannequiins hidden in game by default
)

(fn ConstructionScript
  (if (Editor|InEditor)
    _))
