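; B_LimeRock_Mannequin_FishStix is a LimeRock mannequin Blueprint that displays an Athena character, presumably FishStix going by its name.
; Its construction script, only when in editor or PIE, resets the character definition and variants using the default character item definition and an empty variant array.
; BeginPlay, ActorBeginOverlap and Tick are unimplemented stubs, so the graph adds no runtime behavior.
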(event EventBeginPlay)

(event Collision|EventActorBeginOverlap
  (params (OtherActor :type Actor%20Object%20Reference)))

(event EventTick
  (params (DeltaSeconds :type Float%20%28single-precision%29)))

(fn ConstructionScript
  (if (Editor|InEditororPIE)
    (Customization|ResetAthenaCharacterDefandVariants (Variables|Default|GetDefaultCharacterItemDefinition) (Utilities|Array|MakeArray))))
